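# Translucent UI material for a rounded background in the Compete tournament tile, tinted by a Color parameter.
# A CoreUI rounded-mask function, driven by size and corner radius parameters, shapes the alpha and is multiplied by the color's alpha.
# The Slate post-buffer 0 sample is contrast-adjusted by EffectWeakness and screen-blended.
# UseBlendScreen drives a lerp that controls this backdrop blend effect.
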Material /Compete/TournamentTile/Materials/M_Compete_BackgroundRounded
graph {
 properties {
  bAutomaticallySetUsageInEditor=false
  blendMode=BLEND_Translucent
  materialDomain=MD_UI
  pixelDepthOffsetMode=PDOM_Legacy
  usedShadingModels=Default Lit
 }
 nodes [
  LinearInterpolate(A=NamedRerouteUsage(), B=NamedRerouteUsage(), Alpha=ScalarParameter() {parameterName=UseBlendScreen})
  n1 = VectorParameter() {parameterName=Size, defaultValue=<rgba 1 1 0 1>}
  n2 = VectorParameter() {group=Corners, parameterName=Corner Radius, sortPriority=2, defaultValue=<rgba 0.5 0.5 0.5 0.5>}
  n3 = VectorParameter() {parameterName=Color, defaultValue=<rgba 1 1 1 1>}
  Multiply(
   A=MaterialFunctionCall(
    Size=AppendVector(A=@n1.R, B=@n1.R)
    Corner Radius=AppendVector(A=@n2.RGB, B=@n2.RGB)
    Radius in Pixels=ScalarParameter() {parameterName=RaiusInPixels}
   ) {materialFunction=&MF_UI_CoreUI_RoundedMask}.Shape Mask
   B=@n3.A
  )
  n4 = MaterialFunctionCall!("UVs (Optional)") {materialFunction=/Engine/Functions/UserInterface/GetSlatePost0}
  n5 = MaterialFunctionCall!(Color=Max(A=Saturate(None=@n4.LinearRGB)) {constB=0.0001}, Contrast=ScalarParameter() {defaultValue=0.4, parameterName=EffectWeakness}) {materialFunction=/Engine/Functions/Engine_MaterialFunctions01/ImageAdjustment/Contrast_Preserve_Color}
  NamedRerouteDeclaration(None=MaterialFunctionCall!(Blend=Saturate(None=@n5.Result), Base=NamedRerouteUsage()) {materialFunction=/Engine/Functions/Engine_MaterialFunctions03/Blends/Blend_Screen}.Result) {name=BlendScreen, nodeColor=<rgba 1 0 0.7031 1>}
  NamedRerouteDeclaration(None=@n3.RGB) {name=Color, nodeColor=<rgba 0 1 0.9766 1>}
 ]
}
