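# Base tournament tile button, built on UIKitModularButton with no widgets of its own, that drives visual state transitions on its modular blocks.
# PreConstruct sets block timings; hover, press and selection events request matching states.
# Selection can snap instantly via a one-shot flag; otherwise hover or ShouldFocusOnSelected (false here) picks focused variants for select and deselect.
# Set Blocks Is Selected always passes false to blocks, ignoring its input.
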WidgetBlueprint /Compete/TournamentTile/WBP_TournamentTile_Base
graphs [
 EventGraph {nodes [
   n1 = Event PreConstruct()
   > Parent: PreConstruct(IsDesignTime=@n1.IsDesignTime)
   > n2 = Set BlockTiming(BlockTiming:UIKit Block Timing Structure=Make UIKit Block Timing!(
     Hovering:float=0
     Unhovering:float=0
     Focusing:float=0.3
     Unfocusing:float=0.15
     Pressing:float=0.1
     Releasing:float=0.1
     Disabling:float=0.1
     Enabling:float=0.1
     Selecting:float=0.2
     DeselectingIdle:float=0.2
     DeselectingFocused:float=0.2
    ) -> (UIKitBlockTiming:UIKit Block Timing Structure) {is=title K2Node_MakeStruct}) -> (Output_Get:UIKit Block Timing Structure)
   > SetBlocksTiming!(execute, self:UIKit Modular Button Object Reference, Timing:"UIKit Block Timing Structure (by ref)"=@n2) -> (then)
   Event On Deselected()
   > Set Blocks Is Selected!(execute, self, IsSelected:bool=false) -> (then) {is=title K2Node_CallFunction}
   n3 = IsHovered()
   n4 = ShouldFocusOnSelected!() -> (Result:bool)
   > RequestTransitionOnBlocks!(
    execute
    self:UIKit Modular Button Object Reference
    State:EUIKitBlockVisualState Enum=Select(Option 0:EUIKitBlockVisualState Enum=DeselectedIdle, Option 1:EUIKitBlockVisualState Enum=DeselectedFocused, Index=OR Boolean(A=@n3, B=@n4)) -> (ReturnValue:EUIKitBlockVisualState Enum)
   ) -> (then)
   > Branch(Condition=ShouldFocusOnSelected())
   > RequestTransitionOnBlocks(State=Unfocused)
   Event On Hovered()
   > RequestTransitionOnBlocks(State=Focused)
   Event On Unhovered()
   > RequestTransitionOnBlocks(State=Unfocused)
   Event On Pressed() {is=title K2Node_Event}
   > RequestTransitionOnBlocks(State=Pressed)
   Event On Released() {is=title K2Node_Event}
   > RequestTransitionOnBlocks(State=Released)
   Event On Selected()
   > Set Blocks Is Selected(IsSelected=true)
   > n5 = Branch(Condition=Get IsInstantSelection() -> (IsInstantSelection:bool))
   > RequestInstantTransitionOnBlocks!(execute, self:UIKit Modular Button Object Reference, State:EUIKitBlockInstantTransitionState Enum=InstantSelected) -> (then)
   > Set IsInstantSelection(IsInstantSelection:bool=false) -> (Output_Get:bool)
   n6 = IsHovered()
   n7 = ShouldFocusOnSelected()
   RequestTransitionOnBlocks(execute=@n5.else, State=Select(Option 0=Selected, Option 1=Focused, Index=OR Boolean(A=@n6, B=@n7)))
   Event Construct()
   Event Tick()
  ]}
 GetBlockTiming {nodes [
   GetBlockTiming() {is=title K2Node_FunctionEntry}
   > Return Node!(execute, BlockTiming:UIKit Block Timing Structure=Get BlockTiming() -> (BlockTiming:UIKit Block Timing Structure))
  ]}
 Set Blocks Is Selected {nodes [
   Set Blocks Is Selected!() -> (then, IsSelected:bool=~) {is=title K2Node_FunctionEntry}
   > n1 = For Each Loop(Array:Array of UIKit Block Object References=Get ModularBlocks() -> (ModularBlocks:Array of UIKit Block Object References)) -> (Array Element:UIKit Block Object Reference)
   > n2 = Cast To WBP_UIKit_Block_Base!(execute, Object=@n1) -> (then, CastFailed, AsWBP UIKit Block Base:WBP UIKit Block Base Object Reference) {is=title K2Node_DynamicCast}
   > SetIsSelected!(execute, self:WBP UIKit Block Base Object Reference=@n2, IsSelected:bool=false) -> (then)
  ]}
 ShouldFocusOnSelected {nodes [
   ShouldFocusOnSelected() {is=title K2Node_FunctionEntry}
   > Return Node!(execute, Result:bool=false)
  ]}
]
variables [
 BlockTiming {
  type=UIKitBlockTiming
  default {
   deselectingFocused=0
   deselectingIdle=0
   disabling=0
   enabling=0
   focusing=0
   hovering=0
   pressing=0
   releasing=0
   selecting=0
   unfocusing=0
   unhovering=0
  }
 }
 IsInstantSelection {type=boolean}
]
widgets {info {inheritedWidgetCount=0, parentClass=/Script/FortniteUI.UIKitModularButton, rootWidgetClass=null, widgetCount=0}, widgets []}
